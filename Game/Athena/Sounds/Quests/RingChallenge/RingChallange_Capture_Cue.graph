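# Sound cue for the capture sound in the Athena Ring Challenge quest.
# A single wave player feeds a modulator that applies a fixed volume boost with no pitch variation, offset by a reduced cue volume multiplier.
# It uses the non-critical in-game UI sound class but overrides attenuation for distance falloff.
# A concurrency override allows one instance at a time, stopping the oldest.
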SoundCue /Game/Athena/Sounds/Quests/RingChallenge/RingChallange_Capture_Cue
root=Modulator_0
nodes [
 {id=Modulator_0, children [WavePlayer_0], properties {pitchMax=1, pitchMin=1, volumeMax=1.4, volumeMin=1.4}}
 {id=WavePlayer_0, properties {soundWaveAssetPtr=&/RingChallange_Capture_01}}
]
properties {
 bOverrideAttenuation=true
 bOverrideConcurrency=true
 maxDistance=3400
 soundClassObject=&Fort_UI_InGame_NonCritical
 volumeMultiplier=0.45
 attenuationOverrides {falloffDistance=2400, attenuationShapeExtents {x=1000}}
 concurrencyOverrides {maxCount=1, resolutionRule=StopOldest}
}
