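# Sound cue for the looping spin sound of the ring challenge quest, classed as non-critical in-game UI audio.
# It loops a single wave through one wave player, with no mixing or randomization.
# An attenuation asset handles distance falloff, and an overridden concurrency rule allows only one instance at a time, stopping the oldest when a new one starts.
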SoundCue /Game/Athena/Sounds/Quests/RingChallenge/RingChallange_SpinLoop_Cue
root=WavePlayer_2
nodes [{id=WavePlayer_2, properties {bLooping=true, soundWaveAssetPtr=&/RingChallange_SpinLoop_02}}]
properties {
 attenuationSettings=&/Ringchallenge_Att
 bOverrideConcurrency=true
 maxDistance=1800
 soundClassObject=&Fort_UI_InGame_NonCritical
 volumeMultiplier=0.55
 attenuationOverrides {falloffDistance=1700, attenuationShapeExtents {x=800}}
 concurrencyOverrides {maxCount=1, resolutionRule=StopOldest}
}
